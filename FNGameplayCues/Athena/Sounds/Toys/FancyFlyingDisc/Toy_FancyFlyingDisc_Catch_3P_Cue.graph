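# Third-person sound cue for catching the Fancy Flying Disc toy.
# A mixer layers a random, volume-boosted catch sound with a quieter, delayed random score pop borrowed from LTM audio.
# It uses toy-hit attenuation and the emote music sound class, with overridden concurrency that limits it to one instance per owner.
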SoundCue /FNGameplayCues/Athena/Sounds/Toys/FancyFlyingDisc/Toy_FancyFlyingDisc_Catch_3P_Cue
root=Mixer_0
nodes [
 {id=Delay_0, children [Random_2], properties {delayMax=0.45, delayMin=0.45}}
 {id=Mixer_0, children [Delay_0,Modulator_0], properties {inputVolume [0.5,1]}}
 {id=Modulator_0, children [Random_0], properties {volumeMax=1.3, volumeMin=1.2}}
 {id=Random_0, children [WavePlayer_0,WavePlayer_1,WavePlayer_8], properties {weights [1,1,1]}}
 {id=Random_2, children [WavePlayer_5,WavePlayer_6,WavePlayer_7], properties {weights [1,1,1]}}
 {id=WavePlayer_0, properties {soundWaveAssetPtr=&/Athena_Toy_FancyFlyingDisc_Catch_01}}
 {id=WavePlayer_1, properties {soundWaveAssetPtr=&/Athena_Toy_FancyFlyingDisc_Catch_02}}
 {id=WavePlayer_5, properties {soundWaveAssetPtr=/Game/Athena/Sounds/LTMs/Score/ScoreLTM_InGamePop_Little_01}}
 {id=WavePlayer_6, properties {soundWaveAssetPtr=/Game/Athena/Sounds/LTMs/Score/ScoreLTM_InGamePop_Little_02}}
 {id=WavePlayer_7, properties {soundWaveAssetPtr=/Game/Athena/Sounds/LTMs/Score/ScoreLTM_InGamePop_Little_03}}
 {id=WavePlayer_8, properties {soundWaveAssetPtr=&/Athena_Toy_FancyFlyingDisc_Catch_03}}
]
properties {
 attenuationSettings=/Game/Sounds/Attenuation/Fort_ToyHit_Att
 bOverrideConcurrency=true
 maxDistance=2800
 soundClassObject=&Fort_Music_Emote
 concurrencyOverrides {bLimitToOwner=true, maxCount=1, resolutionRule=StopLowestPriorityThenPreventNew, volumeScaleAttackTime=0.3}
}
